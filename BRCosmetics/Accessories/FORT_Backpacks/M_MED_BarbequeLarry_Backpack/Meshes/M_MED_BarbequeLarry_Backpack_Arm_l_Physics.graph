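# Physics asset for the left arm piece of the BarbequeLarry backpack cosmetic, setting up physics bodies for its flap and shoulder.
# A box covers flap_l, a tiny capsule sits on its flap_l_MSTR bone, and a longer capsule covers shoulder_l.
# A single constraint joins the flap to flap_l_MSTR as a hinge, with twist and swing 1 locked and only swing 2 left to rotate.
# The shoulder body has no constraint.
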PhysicsAsset /BRCosmetics/Accessories/FORT_Backpacks/M_MED_BarbequeLarry_Backpack/Meshes/M_MED_BarbequeLarry_Backpack_Arm_l_Physics
bodies {
 flap_l [{shapeType=Box, extentX=9.6577, extentY=1.7397, extentZ=10.0572, center=<4 0 0>, rotation=<rot 0 0 -90>}]
 flap_l_MSTR [{shapeType=Capsule, length=0.505, radius=0.505, rotation=<rot 0 -90 0>}]
 shoulder_l [{shapeType=Capsule, length=8.9165, radius=2.1578, center=<4.513 0 0>, rotation=<rot 90 0 0>}]
}
constraints [{
  bone1Name=flap_l
  bone2Name=flap_l_MSTR
  swing1LimitDegrees=0
  swing1Motion=Locked
  swing2LimitDegrees=95
  twistLimitDegrees=0
  twistMotion=Locked
 }]
