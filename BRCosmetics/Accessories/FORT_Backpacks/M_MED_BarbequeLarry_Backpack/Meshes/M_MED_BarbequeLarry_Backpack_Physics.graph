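# Physics asset for the Barbeque Larry backpack cosmetic, giving its body, shoulder, flap and antenna bones collision shapes and joints for physics motion.
# A central body box anchors a three-link antenna capsule chain with limited swing and twist, two shoulder capsules and two small flap master capsules.
# Shoulder and master joints rotate freely, while each box-shaped flap hinges on its master through a single limited swing axis, with the other axes locked.
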PhysicsAsset /BRCosmetics/Accessories/FORT_Backpacks/M_MED_BarbequeLarry_Backpack/Meshes/M_MED_BarbequeLarry_Backpack_Physics
bodies {
 antennae_01 [{shapeType=Capsule, length=1, radius=0.5, center=<0.1 0 0>, rotation=<rot -90 0 0>}]
 antennae_02 [{shapeType=Capsule, length=1, radius=0.5, center=<0.1 0 0>, rotation=<rot -90 0 0>}]
 antennae_03 [{shapeType=Capsule, length=1, radius=0.5, center=<0.1 0 0>, rotation=<rot -90 0 0>}]
 body [{shapeType=Box, extentX=11.4116, extentY=22.7712, extentZ=1.7672, center=<14.328 -1.7434 -5.1111>}]
 flap_l [{shapeType=Box, extentX=9.6577, extentY=3.0394, extentZ=14.5017, center=<4 0 0>, rotation=<rot 0 0 -90>}]
 flap_l_MSTR [{shapeType=Capsule, length=0.505, radius=0.505, rotation=<rot 0 -90 0>}]
 flap_r [{shapeType=Box, extentX=9.6577, extentY=1.8444, extentZ=13.0014, center=<4 0 0>, rotation=<rot 0 0 -90>}]
 flap_r_MSTR [{shapeType=Capsule, length=0.505, radius=0.505, rotation=<rot 0 -90 0>}]
 shoulder_l [{shapeType=Capsule, length=9.3667, radius=2.1578, center=<4.513 0 0>, rotation=<rot 90 0 0>}]
 shoulder_r [{shapeType=Capsule, length=6.607, radius=2.209, center=<-4.5087 0.0001 0>, rotation=<rot 90 0 0>}]
}
constraints [
 {bone1Name=antennae_01, bone2Name=body, swing1LimitDegrees=45, swing2LimitDegrees=45, twistLimitDegrees=15}
 {bone1Name=antennae_02, bone2Name=antennae_01, swing1LimitDegrees=45, swing2LimitDegrees=45, twistLimitDegrees=15}
 {bone1Name=antennae_03, bone2Name=antennae_02, swing1LimitDegrees=45, swing2LimitDegrees=45, twistLimitDegrees=15}
 {
  bone1Name=shoulder_l
  bone2Name=body
  swing1LimitDegrees=45
  swing1Motion=Free
  swing2LimitDegrees=45
  swing2Motion=Free
  twistLimitDegrees=45
  twistMotion=Free
 }
 {
  bone1Name=shoulder_r
  bone2Name=body
  swing1LimitDegrees=45
  swing1Motion=Free
  swing2LimitDegrees=45
  swing2Motion=Free
  twistLimitDegrees=45
  twistMotion=Free
 }
 {
  bone1Name=flap_l_MSTR
  bone2Name=body
  swing1LimitDegrees=45
  swing1Motion=Free
  swing2LimitDegrees=45
  swing2Motion=Free
  twistLimitDegrees=45
  twistMotion=Free
 }
 {
  bone1Name=flap_r_MSTR
  bone2Name=body
  swing1LimitDegrees=45
  swing1Motion=Free
  swing2LimitDegrees=45
  swing2Motion=Free
  twistLimitDegrees=45
  twistMotion=Free
 }
 {
  bone1Name=flap_r
  bone2Name=flap_r_MSTR
  swing1LimitDegrees=0
  swing1Motion=Locked
  swing2LimitDegrees=95
  twistLimitDegrees=0
  twistMotion=Locked
 }
 {
  bone1Name=flap_l
  bone2Name=flap_l_MSTR
  swing1LimitDegrees=0
  swing1Motion=Locked
  swing2LimitDegrees=95
  twistLimitDegrees=0
  twistMotion=Locked
 }
]
