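# Animation Blueprint for the BarbequeLarry backpack, adding secondary motion and shoulder-driven flap poses to its input pose.
# A body AnimDynamics spring scales with pawn speed (off while falling), then shoulder rotations drive additive flap transforms.
# Two RigidBody passes follow, enabled when the pet AnimBP's lobby-idle and idle-break curves are equal.
# Parachute opening resets dynamics; the flap AnimDynamics and sequence players are unconnected to the output.
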AnimBlueprint /BRCosmetics/Accessories/FORT_Backpacks/M_MED_BarbequeLarry_Backpack/Meshes/M_MED_BarbequeLarry_AnimBP
target_skeleton=/Game/Accessories/FORT_Backpacks/M_MED_BarbequeLarry_Backpack/Meshes/M_MED_BarbequeLarry_Backpack_Skeleton
graphs [
 AnimGraph {nodes [
   n1 = Anim Dynamics - Bone: flap_l(Alpha=1, ExternalForce="0,0,0") {
    bGravityOverrideInSimSpace=true
    bUseGravityOverride=true
    alphaBoolBlend {blendOption=Linear}
    boundBone {boneName=flap_l}
    componentLinearAccScale=<0.5 0.5 0.5>
    gravityOverride {z=-980}
    physicsBodyDefinitions [{
      sphereCollisionRadius=0
      boundBone {boneName=flap_l}
      boxExtents=<5 5 1>
      constraintSetup {angularLimitsMax=<0 10 0>, angularLimitsMin=<0 -50 0>, angularTarget=<0 0 0>}
      localJointOffset=<0.75 0 0>
     }]
   }
   Anim Dynamics - Bone: flap_r(ComponentPose=@n1, Alpha=1, ExternalForce="0,0,0") {
    bGravityOverrideInSimSpace=true
    bUseGravityOverride=true
    alphaBoolBlend {blendOption=Linear}
    boundBone {boneName=flap_r}
    componentLinearAccScale=<0.5 0.5 0.5>
    gravityOverride {z=-980}
    physicsBodyDefinitions [{
      sphereCollisionRadius=0
      boundBone {boneName=flap_r}
      boxExtents=<5 5 1>
      constraintSetup {angularLimitsMax=<0 10 0>, angularLimitsMin=<0 -50 0>, angularTarget=<0 0 0>}
      localJointOffset=<-0.75 0 0>
     }]
   }
   AnimGraphNode_SubInput_0 = Input Pose()
   n2 = Anim Dynamics - Bone: body(
    ComponentPose=Local To Component(LocalPose=@AnimGraphNode_SubInput_0)
    Alpha=Get bodyDynamics() -> (bodyDynamics:double)
    ExternalForce="0,0,0"
   ) {
    bGravityOverrideInSimSpace=true
    bLinearSpring=true
    bUseGravityOverride=true
    linearDampingOverride=0.99
    linearSpringConstant=100
    alphaBoolBlend {blendOption=Linear}
    alphaScaleBiasClamp {bInterpResult=true, bMapRange=true, inRange {max=500}}
    boundBone {boneName=body}
    componentAppliedLinearAccClamp=<5000 5000 5000>
    componentLinearAccScale=<0.5 0.5 0.5>
    componentLinearVelScale=<0.5 0.5 0.5>
    gravityOverride {z=-980}
    physicsBodyDefinitions [{
      sphereCollisionRadius=0
      boundBone {boneName=body}
      boxExtents=<15 10 15>
      constraintSetup {angularLimitsMax=<0 0 0>, angularLimitsMin=<0 0 0>, angularTarget=<0 0 0>, linearAxesMax=<0.5 0 0.5>, linearAxesMin=<-0.5 0 -0.5>}
      localJointOffset=<0.75 0 0>
     }]
   }
   n3 = "Transform (Modify) Bone - Bone: flap_l_MSTR"(ComponentPose=@n2, Alpha=Get arm_l_up() -> (arm_l_up:double), Translation="1.2,0,1.2", Rotation="-90,0,0", Scale="1,1,1") {
    rotationMode=BMM_Additive
    rotationSpace=BCS_ParentBoneSpace
    translationMode=BMM_Additive
    alphaBoolBlend {blendOption=Linear}
    boneToModify {boneName=flap_l_MSTR}
    rotation {pitch=-90}
    translation {x=1.2, z=1.2}
   }
   n4 = "Transform (Modify) Bone - Bone: flap_r_MSTR"(ComponentPose=@n3, Alpha=Get arm_r_up() -> (arm_r_up:double), Translation="-1.5,0,1.5", Rotation="90,0,0", Scale="1,1,1") {
    rotationMode=BMM_Additive
    translationMode=BMM_Additive
    alphaBoolBlend {blendOption=Linear}
    boneToModify {boneName=flap_r_MSTR}
    rotation {pitch=90}
    translation {x=-1.5, z=1.5}
   }
   n5 = Break Fort Anim Input Skydiving External Force(FortAnimInput_SkydivingExternalForce=Make Fort Anim Input Skydiving External Force!() -> (FortAnimInput_SkydivingExternalForce:Fort Anim Input Skydiving External Force Structure) {is=title K2Node_MakeStruct}) -> (
    bUseSkydivingVectorForce=~
    SkydivingVectorForce=~
    HeadToPelvisDirection=~
    FloatingMultiplier=~
    FloatingAdditive=~
    DivingMultiplier=~
    DivingAdditive=~
    ParachutingMultiplier=~
    ParachutingAdditive=~
   )
   n6 = "Equal (Float)"!(A:double=Get isPlayingIdleLobby() -> (isPlayingIdleLobby:double), B:double=Get isPlayingIdleBreak2() -> (isPlayingIdleBreak2:double)) -> (ReturnValue:bool) {is=title K2Node_CallFunction}
   n7 = RigidBody!(ComponentPose=@n4, bAlphaBoolEnabled:bool=@n6, ExternalForce=@n5.SkydivingVectorForce) -> (Pose) {
    alphaInputType=Bool
    bOverrideWorldGravity=true
    overridePhysicsAsset=&/M_MED_BarbequeLarry_Backpack_Physics
    simulationSpace=BaseBoneSpace
    alphaBoolBlend {blendOption=Linear}
    alphaScaleBiasClamp {bMapRange=true, outRange {max=0, min=1}}
    baseBoneRef {boneName=root}
    componentAppliedLinearAccClamp=<3000 3000 3000>
    componentLinearAccScale=<0.4 0.3 0.3>
    componentLinearVelScale=<0.4 0.3 0.3>
    overrideWorldGravity {z=-980}
   }
   n8 = RigidBody(ComponentPose=@n7, bAlphaBoolEnabled=@n6, ExternalForce=@n5.SkydivingVectorForce) {
    alphaInputType=Bool
    bOverrideWorldGravity=true
    overridePhysicsAsset=&/M_MED_BarbequeLarry_Backpack_Arm_l_Physics
    simulationSpace=BaseBoneSpace
    alphaBoolBlend {blendOption=Linear}
    alphaScaleBiasClamp {outRange {max=0, min=1}}
    baseBoneRef {boneName=root}
    componentAppliedLinearAccClamp=<3000 3000 3000>
    componentLinearAccScale=<0.4 0.3 0.3>
    componentLinearVelScale=<0.4 0.3 0.3>
    overrideWorldGravity {z=-980}
   }
   Output Pose(Result=Component To Local(ComponentPose=@n8)) {layerGroup=None}
   Sequence Player 'IdleBreak_02'() {sequence=/Game/Animation/Game/Pets/BBQ_Larry/Backpack/Idle/IdleBreak_02}
   Sequence Player 'Idle_Lobby'() {sequence=/Game/Animation/Game/Pets/BBQ_Larry/Backpack/Idle/Idle_Lobby}
  ]}
 EventGraph {nodes [
   Event BlueprintUpdateAnimation()
   > n1 = Branch(Condition=Get bIsFalling())
   > n2 = Set bodyDynamics(bodyDynamics:double=0) -> (Output_Get:double)
   n3 = GetOwningComponent()
   n4 = BreakTransform(InTransform=GetDeltaTransformFromRefPose(self=@n3, BoneName=shoulder_r, BaseName=body))
   Set arm_r_up(execute=[@n2.then, @n8.then], arm_r_up:double=MapRangeClamped(Value=BreakRotator(InRot=@n4.Rotation), InRangeA=0, InRangeB=50)) -> (Output_Get:double)
   n5 = BreakTransform(InTransform=GetDeltaTransformFromRefPose(self=@n3, BoneName=shoulder_l, BaseName=body))
   > Set arm_l_up(arm_l_up:double=MapRangeClamped(Value=BreakRotator(InRot=@n5.Rotation), InRangeA=5, InRangeB=50)) -> (Output_Get:double)
   > n6 = Cast To FortPlayerPet!(execute, Object=GetOwningActor()) -> (then, CastFailed, AsFort Player Pet:Fort Player Pet Object Reference) {is=title K2Node_DynamicCast}
   > n7 = Cast To B_BBQ_Larry_AnimBP!(
    execute
    Object=GetAnimInstance(self=Get PetMesh!(self:Fort Player Pet Object Reference=@n6) -> (PetMesh:Skeletal Mesh Component Budgeted Object Reference))
   ) -> (then, CastFailed, AsB BBQ Larry Anim BP:B BBQ Larry Anim BP Object Reference) {is=title K2Node_DynamicCast}
   > Set isPlayingIdleBreak2(isPlayingIdleBreak2:double=GetCurveValue(self=@n7, CurveName=idleBreak2_Active)) -> (Output_Get:double)
   > Set isPlayingIdleLobby(isPlayingIdleLobby:double=GetCurveValue(self=@n7, CurveName=idleLobby_Active)) -> (Output_Get:double)
   n8 = Set bodyDynamics(execute=@n1.else, bodyDynamics=Get PawnSpeed2D())
   Event OnParachuteOpened()
   > ResetDynamics(InTeleportType=TeleportPhysics)
   > ResetDynamics(InTeleportType=ResetPhysics)
   TryGetPawnOwner()
  ]}
]
variables [
 arm_l_up {type=number}
 arm_r_up {type=number}
 bodyDynamics {type=number, default=1}
 isPlayingIdleBreak2 {type=number}
 isPlayingIdleLobby {type=number}
]
